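# MetaSound Source that plays a marked-opponent HUD caret placement sound as a mono one-shot, stored with the DragonCartNarrative orb scan audio.
# A single Wave Player starts on play, feeds the mono output and signals finish when the wave ends.
# It uses HUD marking attenuation and the critical in-game UI sound class.
# Its concurrency override caps simultaneous instances, stopping the oldest first.
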MetaSoundSource /DragonCartNarrative/Assets/Audio/OrbScan/MSS_DragonCart_Marked_Opponent
inputs [UE.Source.OnPlay {type=Trigger}]
interfaces [UE.OutputFormat.Mono:1.0, UE.Source.OneShot:1.0, UE.Source:1.1]
outputs [
 "UE.OutputFormat.Mono.Audio:0" {type=Audio}
 UE.Source.OneShot.OnFinished {type=Trigger}
]
pages [Default {
  input_defaults {UE.Source.OnPlay=null}
  nodes [WavePlayer_0 = UE.Wave Player:Mono(Play=^UE.Source.OnPlay) {also {defaults {Wave Asset=/Game/Sounds/Fort_UI_Sounds/Gameplay/Marking/HUD_Marked_Opponent_Caret_Placement_01}, version [1,1]}}]
  links [{dst="@UE.Source.OneShot.OnFinished", src=WavePlayer_0.On Finished}, {dst="@UE.OutputFormat.Mono.Audio:0", src=WavePlayer_0.Out Mono}]
 }]
properties {
 assetImportData=&self:AssetImportData
 attenuationSettings=/Game/Sounds/Fort_UI_Sounds/Gameplay/Marking/ATT_HUD_Marked_Opponent_UI
 bOverrideConcurrency=true
 sampleRate=48000
 soundClassObject=&Fort_UI_InGame_Critical
 volume=0.8
 concurrencyOverrides {maxCount=3, resolutionRule=StopOldest}
}
